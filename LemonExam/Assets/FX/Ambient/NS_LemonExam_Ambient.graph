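# Niagara ambient effect that layers short-lived rising particles over a tall vertical glow sprite, likely for a sword.
# The GPU emitter Minimal001 spawns tinted sprites in a cylinder, with drag, randomized material parameters and a brief early-life push away from the vertical axis.
# SwordGlow bursts a persistent Z-aligned sprite with waveform-pulsed alpha and size scaled by owner scale.
# Both emitters multiply alpha by a System.Age curve, delaying their fade-in after activation.
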NiagaraSystem /LemonExam/Assets/FX/Ambient/NS_LemonExam_Ambient
dependencies {usedDataInterfaces [&NiagaraDataInterfaceCurve], usedDynamicInputs [], usedModules [&SystemState]}
emitters {
 Minimal001 {
  simTarget=GPUComputeSim
  emitterUpdateScript [
   &EmitterState(Life Cycle Mode=::NewEnumerator1, Loop Behavior=::NewEnumerator1, Loop Duration Mode=::NewEnumerator1, UseLoopDelay=false)
   &SpawnRate(SpawnRate=45.3327, Use Spawn Probability=false)
  ]
  particleSpawnScript [
   &InitializeParticle(
    Color Maximum=<rgba 0.7664 0.3739 1 1>
    Color Minimum=<rgba 0.3322 0.4625 1 1>
    Color Mode=::NewEnumerator2
    Lifetime Max=0.9661
    Lifetime Min=0.8551
    Lifetime Mode=::NewEnumerator1
    Sprite Size Mode=::NewEnumerator1
    Uniform Sprite Size Max=13.781
    Uniform Sprite Size Min=8.7321
    UsePositionOffset=true
    Write Material Random=true
   )
   &ShapeLocation(
    Cylinder Height=94.3958
    Cylinder Height Midpoint=0.1516
    Cylinder Radius=9.8562
    MaskSpawnGroup=false
    Offset=<0 0 36.2902>
    Offset Mode=::NewEnumerator0
    Override Seed=false
    Shape Primitive=::NewEnumerator1
    Surface Only=false
    UseOwnerScale=true
   )
   &AddVelocity(Velocity=&RandomRangeVector(Minimum=<0 0 5.8471>, Maximum=<0 0 9.6308>))
  ]
  particleUpdateScript [
   &ParticleState()
   &ScaleSpriteSize(Uniform Curve Sprite Scale=DataInterface() {propertyValues {Curve {keys [
        {arriveTangent=0.6845, interpMode=RCIM_Cubic, leaveTangent=0.6845, tangentMode=RCTM_User, time=0.2696, value=1}
        {interpMode=RCIM_Cubic, tangentMode=RCTM_Break, time=1, value=1.25}
       ]}}})
   &DynamicMaterialParameters(
    Index 0 Param 1=&RandomRangeFloat(Minimum=0, Maximum=100)
    Module.Index0Float0Write=true
    Module.Index0Float1Write=true
    Module.Index0Float2Write=true
    Module.Index0Float3Write=true
   )
   &Drag(Drag=&RandomRangeFloat(Minimum=1, Maximum=3), Use Linear Drag=true, Use Rotational Drag=true)
   /Niagara/Modules/Update/Forces/LineAttractionForce(
    Attraction Strength:NiagaraFloat=&FloatFromCurve(
     FloatCurve=DataInterface() {propertyValues {Curve {keys [{interpMode=RCIM_Cubic, tangentMode=RCTM_Break, time=0.1, value=1}, <at 0.2 0>]}}}
     Scale Curve=-20
    )
    Line End:Vector3f=<0 0 1000>
    Line Start:Vector3f=<0 0 0>
    Use Attraction Falloff=false
   )
   &SolveForcesAndVelocity(Clamp Velocity=false, Limit Acceleration=false)
   &ScaleColor(
    Scale Alpha=&Multiply_Float(
     A=&FloatFromCurve(
      FloatCurve=DataInterface() {propertyValues {Curve {
         defaultValue=1
         keys [
          {arriveTangent=0.2741, interpMode=RCIM_Cubic, leaveTangent=0.2741, tangentMode=RCTM_User, time=0, value=0}
          {arriveTangent=0.0495, interpMode=RCIM_Cubic, leaveTangent=0.0495, tangentMode=RCTM_User, time=0.3464, value=0.9847}
          {arriveTangent=-0.0405, interpMode=RCIM_Cubic, leaveTangent=-0.0405, tangentMode=RCTM_User, time=1, value=0}
         ]
        }}}
      Scale Curve=&RandomRangeFloat(Minimum=0.3, Maximum=0.5392)
     )
     B=&FloatFromCurve(
      FloatCurve=DataInterface() {propertyValues {Curve {
         defaultValue=1
         keys [<at 0.5044 0 RCIM_Cubic>, {arriveTangent=3.2567, interpMode=RCIM_Cubic, leaveTangent=3.2567, tangentMode=RCTM_User, time=1, value=1}]
        }}}
      CurveIndex=$System.Age
     )
    )
    ScaleA=true
    ScaleRGB=true
   )
  ]
  rendererClasses [&NiagaraSpriteRendererProperties]
  renderers [{rendererClass=&NiagaraSpriteRendererProperties}]
 }
 SwordGlow {
  emitterUpdateScript [
   &EmitterState(
    Inactive Response=::NewEnumerator1
    Life Cycle Mode=::NewEnumerator1
    Loop Behavior=::NewEnumerator1
    Loop Duration Mode=::NewEnumerator1
    UseLoopDelay=false
   )
   &SpawnBurst_Instantaneous(Use Spawn Probability=false, UseLoopCountLimit=false)
  ]
  particleSpawnScript [&InitializeParticle(
    Color=<rgba 0.6574 0.6574 0.6574 1>
    Position Offset=<0 0 9.9988>
    Sprite Size=<91.0954 194.0351>
    Sprite Size Mode=::NewEnumerator4
    UsePositionOffset=true
    Write Material Random=true
   )]
  particleUpdateScript [
   &ParticleState(Kill Particles When Lifetime Has Elapsed=false, Let Infinitely Lived Particles Die When Emitter Deactivates=true)
   &SpriteFacingAndAlignment(Sprite Alignment=<0 0 1>, Sprite Facing=<0 1 0>, WriteAlignment=true, WriteFacing=false)
   &ScaleColor(
    Scale Alpha=&Multiply_Float(
     A=&FloatFromCurve(
      FloatCurve=DataInterface() {propertyValues {Curve {keys [<at 0 0>, <at 0.0519 0.1612>, {interpMode=RCIM_Cubic, leaveTangent=-3.459, tangentMode=RCTM_Break, time=0.0951, value=1.466}]}}}
      CurveIndex=$Emitter.Age
      Scale Curve=&Waveform("Waveform [1]"=::NewEnumerator0, "[1] Phase Scale"=2.7127, "[1] Frequency"=1, "[1] Amplitude Min/Max"=<0.6405 1>)
     )
     B=&FloatFromCurve(
      FloatCurve=DataInterface() {propertyValues {Curve {
         defaultValue=1
         keys [<at 0.5044 0 RCIM_Cubic>, {arriveTangent=3.2567, interpMode=RCIM_Cubic, leaveTangent=3.2567, tangentMode=RCTM_User, time=1, value=1}]
        }}}
      CurveIndex=$System.Age
     )
    )
    ScaleA=true
    ScaleRGB=true
   )
   &ScaleSpriteSize(
    Scale Sprite Size Mode=::NewEnumerator0
    Uniform Scale Factor=&MakeFloatFromVector(Vector=$Engine.Owner.Scale, Channel=::NewEnumerator0)
   )
  ]
  rendererClasses [&NiagaraSpriteRendererProperties]
  renderers [{rendererClass=&NiagaraSpriteRendererProperties}]
 }
}
summary {
 systemName=NS_LemonExam_Ambient
 emitters [
  {emitterName=Minimal001, simTarget=GPUComputeSim, rendererClasses [&NiagaraSpriteRendererProperties]}
  {emitterName=SwordGlow, rendererClasses [&NiagaraSpriteRendererProperties]}
 ]
 userVariables []
}
